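; State Tree task for the Umbramolt Hole scenario that closes the hole's rift.
; On entering the state it casts the input RiftGate to RiftGate_Umbramolt, adds the UmbraMolt.Hole.Completed gameplay tag to it, then destroys the rift and finishes successfully.
; An invalid gate or failed cast leads to finishing the task as failed.
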(event EventEnterState
  (params (Transition :type State%20Tree%20Transition%20Result%20Structure%20%28by%20ref%29))
  (bind _riftgate (Variables|Input|GetRiftGate))
  (Utilities|IsValid _riftgate
    (:Is%20Valid
      (bind _asrift_gate_umbramolt (Utilities|Casting|CastToRiftGate_Umbramolt _riftgate)
        (:then
          (Class|BPFLAICommonFunctions|AddUniqueGameplayTag _asrift_gate_umbramolt "(TagName=\"UmbraMolt.Hole.Completed\")")
          (Class|RiftGateBase|DestroyRift _asrift_gate_umbramolt)
          (StateTree|FinishTask true))
        (:CastFailed
          _)))
    (:Is%20Not%20Valid
      _))
  (StateTree|FinishTask false))
