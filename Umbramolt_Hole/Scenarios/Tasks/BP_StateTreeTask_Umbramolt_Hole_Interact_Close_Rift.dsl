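; StateTree task for the Umbramolt Hole scenario that closes a rift gate when its state is entered.
; It reads the RiftGate input and, if valid, casts it to RiftGate_Umbramolt_Interact.
; It then adds the UmbraMolt.Hole.Completed gameplay tag to the gate, calls SilentDie on it and finishes successfully.
; An invalid gate reference finishes the task as failed.
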(event EventEnterState
  (params (Transition :type State%20Tree%20Transition%20Result%20Structure%20%28by%20ref%29))
  (bind _riftgate (Variables|Input|GetRiftGate))
  (Utilities|IsValid _riftgate
    (:Is%20Valid
      (bind _asrift_gate_umbramolt_interact (Utilities|Casting|CastToRiftGate_Umbramolt_Interact _riftgate))
      (Class|BPFLAICommonFunctions|AddUniqueGameplayTag _asrift_gate_umbramolt_interact "(TagName=\"UmbraMolt.Hole.Completed\")")
      (Gameplay|SilentDie _asrift_gate_umbramolt_interact false :!class /Script/FortniteGame.BuildingActor)
      (StateTree|FinishTask true))
    (:Is%20Not%20Valid
      (StateTree|FinishTask false))))
