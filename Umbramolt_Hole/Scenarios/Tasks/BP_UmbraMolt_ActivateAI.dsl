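; State Tree task for the Umbramolt rift gate encounter, with an empty enter-state event and a function that positions the twins Gogo and Yuki.
; RotateTheTwins yaws the gate's shared spawn root toward the activating player's location.
; It projects both spawn points onto the navmesh using an avoid-water filter.
; It then teleports each twin to its point, raised by capsule half-height and matching that point's yaw.
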(event EventEnterState
  (params (Transition :type State%20Tree%20Transition%20Result%20Structure%20%28by%20ref%29)))

(fn RotateTheTwins
  (bind (_returnvalue_roll _returnvalue_pitch _returnvalue_yaw) (Math|Rotator|BreakRotator (Transformation|GetWorldRotation (Variables|/Umbramolt_Hole/Gameplay/RiftGate_Umbramolt.RiftGate_Umbramolt_C|GetYukiSpawnPoint (Variables|Default|GetRiftGateRef)))))
  (bind (_returnvalue_roll_2 _returnvalue_pitch_2 _returnvalue_yaw_2) (Math|Rotator|BreakRotator (Transformation|GetWorldRotation (Variables|/Umbramolt_Hole/Gameplay/RiftGate_Umbramolt.RiftGate_Umbramolt_C|GetGogoSpawnPoint (Variables|Default|GetRiftGateRef)))))
  (bind (_returnvalue_roll_3 _returnvalue_pitch_3 _returnvalue_yaw_3) (Math|Rotator|BreakRotator (Math|Rotator|FindLookatRotation (Transformation|GetWorldLocation (Variables|/Umbramolt_Hole/Gameplay/RiftGate_Umbramolt.RiftGate_Umbramolt_C|GetTwinsSpawnPoints (Variables|Default|GetRiftGateRef))) (Variables|Default|GetActivatingPlayerLoc))))
  (Transformation|SetWorldRotation (Variables|/Umbramolt_Hole/Gameplay/RiftGate_Umbramolt.RiftGate_Umbramolt_C|GetTwinsSpawnPoints (Variables|Default|GetRiftGateRef)) (Math|Rotator|MakeRotator 0.0 0.0 _returnvalue_yaw_3) false false)
  (for _array_element (Utilities|Array|MakeArray (Variables|/Umbramolt_Hole/Gameplay/RiftGate_Umbramolt.RiftGate_Umbramolt_C|GetGogoSpawnPoint (Variables|Default|GetRiftGateRef)) (Variables|/Umbramolt_Hole/Gameplay/RiftGate_Umbramolt.RiftGate_Umbramolt_C|GetYukiSpawnPoint (Variables|Default|GetRiftGateRef)))
    (Transformation|SetWorldLocation _array_element (AI|Navigation|ProjectPointtoNavigation (Transformation|GetWorldLocation _array_element) (Fortnite|AI|GetNavigationDataforActor (Variables|Default|GetRiftGateRef)) "/Game/Athena/AI/Phoebe/NavFilter/NavFilter_Phoebe_AvoidWater.NavFilter_Phoebe_AvoidWater_C" "0.000000,0.000000,500.000000") false false))
  (Transformation|SetActorLocationAndRotation (Variables|/Umbramolt_Hole/Gameplay/RiftGate_Umbramolt.RiftGate_Umbramolt_C|GetGogoRef (Variables|Default|GetRiftGateRef)) (Utilities|Operators|Add (Transformation|GetWorldLocation (Variables|/Umbramolt_Hole/Gameplay/RiftGate_Umbramolt.RiftGate_Umbramolt_C|GetGogoSpawnPoint (Variables|Default|GetRiftGateRef))) (Math|Vector|MakeVector 0.0 0.0 (Components|Capsule|GetScaledCapsuleHalfHeight (Variables|/Script/Engine.Character|GetCapsuleComponent (Variables|/Umbramolt_Hole/Gameplay/RiftGate_Umbramolt.RiftGate_Umbramolt_C|GetGogoRef (Variables|Default|GetRiftGateRef)))))) (Math|Rotator|MakeRotator 0.0 0.0 _returnvalue_yaw_2) false true)
  (Transformation|SetActorLocationAndRotation (Variables|/Umbramolt_Hole/Gameplay/RiftGate_Umbramolt.RiftGate_Umbramolt_C|GetYukiRef (Variables|Default|GetRiftGateRef)) (Utilities|Operators|Add (Transformation|GetWorldLocation (Variables|/Umbramolt_Hole/Gameplay/RiftGate_Umbramolt.RiftGate_Umbramolt_C|GetYukiSpawnPoint (Variables|Default|GetRiftGateRef))) (Math|Vector|MakeVector 0.0 0.0 (Components|Capsule|GetScaledCapsuleHalfHeight (Variables|/Script/Engine.Character|GetCapsuleComponent (Variables|/Umbramolt_Hole/Gameplay/RiftGate_Umbramolt.RiftGate_Umbramolt_C|GetYukiRef (Variables|Default|GetRiftGateRef)))))) (Math|Rotator|MakeRotator 0.0 0.0 _returnvalue_yaw) false true))
