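; State Tree task for the Umbramolt interact scenario that activates the twin AI, Gogo and Yuki, at a nearby rift gate.
; On state entry it caches a rift gate found near the scenario and records a nearby player pawn's location, then calls RotateTheTwins, whose body is empty here.
; It adds the StartWalkout gameplay tag to both twins referenced by the gate and finishes the task successfully.
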(event EventEnterState
  (params (Transition :type State%20Tree%20Transition%20Result%20Structure%20%28by%20ref%29))
  (bind _outactors (Fortnite|GetAllActorsOfClassinRange "/Umbramolt_Hole/Gameplay/RiftGate_Umbramolt_Interact.RiftGate_Umbramolt_Interact_C" (Transformation|GetActorLocation (Variables|Default|GetScenarioRef)) 1e+03))
  (Utilities|IsValid (Utilities|Array|Get(acopy) _outactors)
    (:Is%20Valid
      (Variables|Default|SetRiftGateRef (Utilities|Array|Get(acopy) _outactors))
      (bind _outfortplayerpawns (Fortnite|GetFortPlayerPawnsInRange (Transformation|GetActorLocation (Variables|Default|GetScenarioRef)) 1e+03))
      (Utilities|IsValid (Utilities|Array|Get(acopy) _outfortplayerpawns)
        (:Is%20Valid
          (Variables|Default|SetActivatingPlayerLoc (Transformation|GetActorLocation (Utilities|Array|Get(acopy) _outfortplayerpawns)))
          (CallFunction|RotatetheTwins))
        (:Is%20Not%20Valid
          _))
      (for _array_element (Utilities|Array|MakeArray (Variables|/Umbramolt_Hole/Gameplay/RiftGate_Umbramolt_Interact.RiftGate_Umbramolt_Interact_C|GetGogoRef (Variables|Default|GetRiftGateRef)) (Variables|/Umbramolt_Hole/Gameplay/RiftGate_Umbramolt_Interact.RiftGate_Umbramolt_Interact_C|GetYukiRef (Variables|Default|GetRiftGateRef)))
        (Class|BPFLAICommonFunctions|AddUniqueGameplayTag _array_element "(TagName=\"Umbramolt_Fraternal.StartWalkout\")")
        (StateTree|FinishTask true)))))

(fn RotateTheTwins)
